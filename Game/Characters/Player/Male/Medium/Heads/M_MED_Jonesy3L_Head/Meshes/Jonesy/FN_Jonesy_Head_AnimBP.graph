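# Jonesy head Animation Blueprint that drives the face from pose and curves copied from the attached parent mesh.
# The copied pose passes through facial-input enum blends (LiveLink branches only reuse it) into a Jonesy facial PoseAsset, then layers FaceSlot from faceAttach.
# When the playlist uses custom character parts, checked at init, the head bone shrinks to near zero.
# The gameplay-driven face state machine and older slot blends are disconnected from output.
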AnimBlueprint /Game/Characters/Player/Male/Medium/Heads/M_MED_Jonesy3L_Head/Meshes/Jonesy/FN_Jonesy_Head_AnimBP
target_skeleton=/Game/Characters/Player/Common/Fortnite_Base_Head/Fortnite_Base_Head_Export_Skeleton
graphs [
 AnimGraph {nodes [
   n1 = Get MeshToCopy() -> (MeshToCopy:Skeletal Mesh Component Object Reference)
   Blend Poses by bool(
    BlendPose_0=Layered blend per bone(BasePose=Copy Pose From Mesh(SourceMeshComponent=@n1) {bCopyCurves=true}, BlendPoses_0=Slot 'DefaultSlot'()) {layerSetup [{branchFilters [{boneName=faceAttach}]}]}
    BlendPose_1=Layered blend per bone(BasePose=Copy Pose From Mesh(SourceMeshComponent=@n1) {bCopyCurves=true}, BlendPoses_0=Slot 'inGameSlot'() {slotName=inGameSlot}) {layerSetup [{branchFilters [{boneName=faceAttach}]}]}
    bActiveValue=Get bCanPlayCustomAnimations() -> (bCanPlayCustomAnimations:bool)
   ) {blendType=Linear}
   n2 = Sequence Player 'Fall'() {sequence=/Game/Animation/Game/MainPlayer/Facial/DefaultStates/Fall}
   AnimGraphNode_StateMachine_0 = Face State Machine() {is=title AnimGraphNode_StateMachine}
   n3 = Get bIsOnGround_0() -> (bIsOnGround_0:bool)
   Blend Poses by bool(
    BlendPose_0=Sequence Player 'DownSights'() {sequence=/Game/Animation/Game/MainPlayer/Facial/DefaultStates/DownSights}
    BlendPose_1=Blend Poses by bool(
     BlendPose_0=Sequence Player 'NonTargetedFire'() {sequence=/Game/Animation/Game/MainPlayer/Facial/DefaultStates/NonTargetedFire}
     BlendPose_1=Blend Poses by bool(BlendPose_0=@AnimGraphNode_StateMachine_0, BlendPose_1=@n2, BlendTime_0=0.2, BlendTime_1=0.2, bActiveValue=@n3) {blendType=Linear}
     BlendTime_0=0.2
     BlendTime_1=0.2
     bActiveValue=Get bRecentlyFired_0() -> (bRecentlyFired_0:bool)
    ) {blendType=Linear}
    BlendTime_0=0.2
    BlendTime_1=0.2
    bActiveValue=Get bIsTargeting_0() -> (bIsTargeting_0:bool)
   ) {blendType=Linear}
   n4 = "Transform (Modify) Bone - Bone: head"!(ComponentPose=Local To Component(LocalPose=Use cached pose 'preScaleCheck'()), Alpha=1) -> (Pose) {scaleMode=BMM_Replace, alphaBoolBlend {blendOption=Linear}, boneToModify {boneName=head}, scale=<0.01 0.01 0.01>}
   Output Pose(Result=Blend Poses by bool(
     BlendPose_0=Component To Local(ComponentPose=@n4)
     BlendPose_1=Use cached pose 'preScaleCheck'()
     BlendTime_0=0
     BlendTime_1=0
     bActiveValue=Get GetPlayerListUsesCustomCharacterParts() -> (GetPlayerListUsesCustomCharacterParts:bool)
    ) {blendType=Linear}) {layerGroup=None}
   Save cached pose 'IncomingPose'(Pose=Copy Pose From Mesh() {bCopyCurves=true, bUseAttachedParent=true}) {cacheName=IncomingPose}
   n5 = "Blend Poses (EFortFacialAnimTypes)"!(
    BlendPose_0:Pose Link Structure=Two Way Blend!(A=Use cached pose 'IncomingPose'(), B=Use cached pose 'IncomingPose'(), AlphaCurveName:Name=DisableFaceOverride) -> (Pose) {blendNode {alphaInputType=Curve, alphaBoolBlend {blendOption=Linear}}}
    BlendPose_1:Pose Link Structure=Use cached pose 'IncomingPose'()
    BlendPose_2:Pose Link Structure=Use cached pose 'IncomingPose'()
    BlendTime_0:float=0.1
    BlendTime_1:float=0.1
    BlendTime_2:float=0.1
    ActiveEnumValue:Byte=Get!() -> (
     FacialInput_SubjectName:Live Link Subject Name Structure
     FacialInput_CurrentAnimType:EFortFacialAnimTypes Enum
     FacialInput_bCurvesOnly:bool=~
    ).FacialInput_CurrentAnimType
   ) -> (Pose:Pose Link Structure) {
    is=title AnimGraphNode_BlendListByEnum
    blendType=Linear
    blendPose [{linkId=-1, sourceLinkId=-1}, {linkId=-1, sourceLinkId=-1}, {linkId=-1, sourceLinkId=-1}]
    blendTime [0.1,0.1,0.1]
   }
   Save cached pose 'GameFacial'(Pose=@n5) {cacheName=GameFacial}
   n6 = "Blend Poses (EFortFacialAnimTypes)"(
    BlendPose_0=Use cached pose 'GameFacial'()
    BlendPose_1=Layered blend per bone(BasePose=Use cached pose 'GameFacial'(), BlendPoses_0=Use cached pose 'LiveLinkFacial'()) {layerSetup [{branchFilters [{boneName=faceAttach}]}]}
    BlendPose_2=Use cached pose 'LiveLinkFacial'()
    BlendTime_0=0.1
    BlendTime_1=0.1
    BlendTime_2=0.1
    ActiveEnumValue=Get() -> (FacialInput_bCurvesOnly=~).FacialInput_CurrentAnimType
   ) {
    blendType=Linear
    blendPose [{linkId=-1, sourceLinkId=-1}, {linkId=-1, sourceLinkId=-1}, {linkId=-1, sourceLinkId=-1}]
    blendTime [0.1,0.1,0.1]
   }
   n7 = M_MED_CAU_Jonesy_Head_01_Facial_Pose_Export_PoseAsset!(SourcePose:Pose Link Structure=@n6) -> (Pose:Pose Link Structure) {
    is=title AnimGraphNode_PoseBlendNode
    blendOption=Linear
    poseAsset=/Game/Characters/Player/Male/Medium/Heads/M_MED_CAU_Jonesy_Head_01/Meshes/M_MED_CAU_Jonesy_Head_01_Facial_Pose_Export_PoseAsset
   }
   Save cached pose 'PostPoseAsset'(Pose=@n7) {cacheName=PostPoseAsset}
   n8 = Use cached pose 'PostPoseAsset'()
   Save cached pose 'preScaleCheck'(Pose=Layered blend per bone(BasePose=Use cached pose 'PostPoseAsset'(), BlendPoses_0=Slot 'FaceSlot'(Source=@n8) {slotName=FaceSlot}) {layerSetup [{branchFilters [{boneName=faceAttach}]}]}) {cacheName=preScaleCheck}
   n9 = Break Fort Anim Input Facial!(FortAnimInput_Facial:"Fort Anim Input Facial Structure (by ref)"=Get FacialInput() -> (FacialInput:Fort Anim Input Facial Structure)) -> (SubjectName:Live Link Subject Name Structure, CurrentAnimType:EFortFacialAnimTypes Enum, bCurvesOnly:bool=~) {is=title K2Node_BreakStruct}
   Save cached pose 'LiveLinkFacial'(Pose=Blend Poses by bool(BlendPose_0=Use cached pose 'GameFacial'(), BlendPose_1=Use cached pose 'GameFacial'(), bActiveValue=@n9.bCurvesOnly) {blendType=Linear}) {cacheName=LiveLinkFacial}
   Sequence Player 'Fortnite_Base_Head_ROM'() {sequence=/Game/Characters/Player/Common/Fortnite_Base_Head/Anim_Tests/Fortnite_Base_Head_ROM}
   GetPlaylistUsesCustomCharacterParts!() -> (ReturnValue:bool)
   Reroute Node(InputPin:Live Link Subject Name Structure=@n9) -> (OutputPin:Live Link Subject Name Structure)
  ]}
 Face State Machine {
  entry=Movement
  path=AnimGraph.AnimGraphNode_StateMachine_0.Face State Machine
  ids=id
  nodes [
   AnimStateNode_0 = Movement() {is=title AnimStateNode}
   AnimStateNode_2 = Skydive() {is=title AnimStateNode}
   AnimStateNode_3 = DBNO() {is=title AnimStateNode}
   AnimStateNode_4 = MeleeSwing() {is=title AnimStateNode}
   AnimStateTransitionNode_2 = AnimStateTransitionNode()
   AnimStateTransitionNode_3 = AnimStateTransitionNode()
   AnimStateTransitionNode_4 = AnimStateTransitionNode()
   AnimStateTransitionNode_5 = AnimStateTransitionNode()
   AnimStateTransitionNode_6 = AnimStateTransitionNode()
   AnimStateTransitionNode_7 = AnimStateTransitionNode()
  ]
  links [
   {
    dst=Movement
    rule=AnimGraph.AnimGraphNode_StateMachine_0.Face State Machine.AnimStateTransitionNode_5.Transition
    src=DBNO
    via=AnimStateTransitionNode_5
   }
   {
    dst=Movement
    rule=AnimGraph.AnimGraphNode_StateMachine_0.Face State Machine.AnimStateTransitionNode_7.Transition
    src=MeleeSwing
    via=AnimStateTransitionNode_7
   }
   {
    dst=DBNO
    rule=AnimGraph.AnimGraphNode_StateMachine_0.Face State Machine.AnimStateTransitionNode_4.Transition
    src=Movement
    via=AnimStateTransitionNode_4
   }
   {
    dst=MeleeSwing
    rule=AnimGraph.AnimGraphNode_StateMachine_0.Face State Machine.AnimStateTransitionNode_6.Transition
    src=Movement
    via=AnimStateTransitionNode_6
   }
   {
    dst=Skydive
    rule=AnimGraph.AnimGraphNode_StateMachine_0.Face State Machine.AnimStateTransitionNode_2.Transition
    src=Movement
    via=AnimStateTransitionNode_2
   }
   {
    dst=Movement
    rule=AnimGraph.AnimGraphNode_StateMachine_0.Face State Machine.AnimStateTransitionNode_3.Transition
    src=Skydive
    via=AnimStateTransitionNode_3
   }
  ]
 }
 Movement {
  path=AnimGraph.AnimGraphNode_StateMachine_0.Face State Machine.AnimStateNode_0.Movement
  nodes [
   n1 = Sequence Player 'Moving'() {sequence=/Game/Animation/Game/MainPlayer/Facial/DefaultStates/Moving}
   n2 = Sequence Player 'Crouch'() {sequence=/Game/Animation/Game/MainPlayer/Facial/DefaultStates/Crouch}
   n3 = Get bIsMoving() -> (bIsMoving:bool)
   n4 = Sequence Player 'Idle'() {sequence=/Game/Animation/Game/MainPlayer/Facial/DefaultStates/Idle}
   n5 = Sequence Player 'Moving'() {sequence=&Moving}
   n6 = Get bIsMoving()
   Output Animation Pose(Result=Blend Poses by bool(
     BlendPose_0=Blend Poses by bool(BlendPose_0=@n1, BlendPose_1=@n2, BlendTime_0=0.2, BlendTime_1=0.2, bActiveValue=@n3) {blendType=Linear}
     BlendPose_1=Blend Poses by bool(BlendPose_0=@n5, BlendPose_1=@n4, BlendTime_0=0.2, BlendTime_1=0.2, bActiveValue=@n6) {blendType=Linear}
     BlendTime_0=0.2
     BlendTime_1=0.2
     bActiveValue=Get bIsCrouching_0() -> (bIsCrouching_0:bool)
    ) {blendType=Linear})
  ]
 }
 Skydive {
  path=AnimGraph.AnimGraphNode_StateMachine_0.Face State Machine.AnimStateNode_2.Skydive
  nodes [Output Animation Pose(Result=Blend Poses by bool(
     BlendPose_0=Sequence Player 'Dive'() {sequence=/Game/Animation/Game/MainPlayer/Facial/DefaultStates/Dive}
     BlendPose_1=Sequence Player 'Glide'() {sequence=/Game/Animation/Game/MainPlayer/Facial/DefaultStates/Glide}
     bActiveValue=Get IsDive0() -> (IsDive0:bool)
    ) {blendType=Linear})]
 }
 DBNO {
  path=AnimGraph.AnimGraphNode_StateMachine_0.Face State Machine.AnimStateNode_3.DBNO
  nodes [Output Animation Pose(Result=Sequence Player 'DBNO'() {sequence=/Game/Animation/Game/MainPlayer/Facial/DefaultStates/DBNO})]
 }
 MeleeSwing {
  path=AnimGraph.AnimGraphNode_StateMachine_0.Face State Machine.AnimStateNode_4.MeleeSwing
  nodes [Output Animation Pose(Result=Sequence Player 'MeleeSwing'() {sequence=/Game/Animation/Game/MainPlayer/Facial/DefaultStates/MeleeSwing})]
 }
 Transition {
  path=AnimGraph.AnimGraphNode_StateMachine_0.Face State Machine.AnimStateTransitionNode_2.Transition
  nodes [Result(bCanEnterTransition=Get IsSkydiving0() -> (IsSkydiving0:bool))]
 }
 Transition {
  path=AnimGraph.AnimGraphNode_StateMachine_0.Face State Machine.AnimStateTransitionNode_3.Transition
  nodes [Result(bCanEnterTransition=NOT Boolean(A=Get IsSkydiving0()))]
 }
 Transition {
  path=AnimGraph.AnimGraphNode_StateMachine_0.Face State Machine.AnimStateTransitionNode_4.Transition
  nodes [Result(bCanEnterTransition=Get bIsDBNO_0() -> (bIsDBNO_0:bool))]
 }
 Transition {
  path=AnimGraph.AnimGraphNode_StateMachine_0.Face State Machine.AnimStateTransitionNode_5.Transition
  nodes [Result(bCanEnterTransition=NOT Boolean(A=Get bIsDBNO_0()))]
 }
 Transition {
  path=AnimGraph.AnimGraphNode_StateMachine_0.Face State Machine.AnimStateTransitionNode_6.Transition
  nodes [Result(bCanEnterTransition=Get bIsMeleeSwing_0() -> (bIsMeleeSwing_0:bool))]
 }
 Transition {
  path=AnimGraph.AnimGraphNode_StateMachine_0.Face State Machine.AnimStateTransitionNode_7.Transition
  nodes [Result(bCanEnterTransition=NOT Boolean(A=Get bIsMeleeSwing_0()))]
 }
 EventGraph {nodes [
   Event BlueprintInitializeAnimation()
   > Set MeshToCopy(MeshToCopy:Skeletal Mesh Component Object Reference=GetOwnerSkeletalMeshComponent()) -> (Output_Get:Skeletal Mesh Component Object Reference)
   > Set GetPlayerListUsesCustomCharacterParts(GetPlayerListUsesCustomCharacterParts:bool=GetPlaylistUsesCustomCharacterParts()) -> (Output_Get:bool)
   TryGetPawnOwner()
   Event BlueprintUpdateAnimation()
  ]}
]
variables [
 GetPlayerListUsesCustomCharacterParts {type=boolean}
 MeshToCopy {type=SkeletalMeshComponent}
]
